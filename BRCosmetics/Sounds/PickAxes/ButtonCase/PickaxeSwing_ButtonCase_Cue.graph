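# Sound cue for the ButtonCase pickaxe swing, mixing a close layer and a medium-range layer, each with its own attenuation.
# The close layer switches on the SwingIdx parameter between random Short and Long ButtonCase swing waves, mostly through pitch modulators.
# The medium layer randomly plays generic distant fire axe swings.
# A shared pickaxe property sheet, melee sound class and melee concurrency set apply to the whole cue.
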SoundCue /BRCosmetics/Sounds/PickAxes/ButtonCase/PickaxeSwing_ButtonCase_Cue
root=Mixer_0
nodes [
 {id=Attenuation_0, children [Modulator_0], properties {attenuationSettings=/Game/Sounds/Attenuation/Fort_Weapon_Pickaxe_Close_Att}}
 {id=Attenuation_1, children [Modulator_1], properties {attenuationSettings=/Game/Sounds/Attenuation/Fort_Weapon_Pickaxe_Medium_Att}}
 {id=Mixer_0, children [Attenuation_0, Attenuation_1], properties {inputVolume [1,1]}}
 {id=Modulator_0, children [Switch_1]}
 {id=Modulator_1, children [Random_1]}
 {id=Modulator_4, children [Random_6], properties {pitchMax=1.1, pitchMin=0.9, volumeMax=0.95, volumeMin=0.85}}
 {id=Modulator_5, children [Random_2], properties {pitchMax=1.1}}
 {id=Random_1, children [WavePlayer_4,WavePlayer_5,WavePlayer_6,WavePlayer_7,WavePlayer_8,WavePlayer_9], properties {weights [1,1,1,1,1,1]}}
 {id=Random_2, children [WavePlayer_18, WavePlayer_19, WavePlayer_20, WavePlayer_21], properties {weights [1,1,1,1]}}
 {id=Random_6, children [WavePlayer_22, WavePlayer_23, WavePlayer_24, WavePlayer_25], properties {weights [1,1,1,1]}}
 {id=Switch_1, children [Random_6,Modulator_4,Modulator_4,Modulator_5,Modulator_5], properties {intParameterName=SwingIdx}}
 {id=WavePlayer_18, properties {soundWaveAssetPtr=&/IP_Pickaxe_ButtonCase_Swings_Long_01}}
 {id=WavePlayer_19, properties {soundWaveAssetPtr=&/IP_Pickaxe_ButtonCase_Swings_Long_02}}
 {id=WavePlayer_20, properties {soundWaveAssetPtr=&/IP_Pickaxe_ButtonCase_Swings_Long_03}}
 {id=WavePlayer_21, properties {soundWaveAssetPtr=&/IP_Pickaxe_ButtonCase_Swings_Long_04}}
 {id=WavePlayer_22, properties {soundWaveAssetPtr=&/IP_Pickaxe_ButtonCase_Swings_Short_01}}
 {id=WavePlayer_23, properties {soundWaveAssetPtr=&/IP_Pickaxe_ButtonCase_Swings_Short_02}}
 {id=WavePlayer_24, properties {soundWaveAssetPtr=&/IP_Pickaxe_ButtonCase_Swings_Short_03}}
 {id=WavePlayer_25, properties {soundWaveAssetPtr=&/IP_Pickaxe_ButtonCase_Swings_Short_04}}
 {id=WavePlayer_4, properties {soundWaveAssetPtr=/Game/Athena/Sounds/Weapons/Fire_Axe_Swing_Distant_01}}
 {id=WavePlayer_5, properties {soundWaveAssetPtr=/Game/Athena/Sounds/Weapons/Fire_Axe_Swing_Distant_02}}
 {id=WavePlayer_6, properties {soundWaveAssetPtr=/Game/Athena/Sounds/Weapons/Fire_Axe_Swing_Distant_03}}
 {id=WavePlayer_7, properties {soundWaveAssetPtr=/Game/Athena/Sounds/Weapons/Fire_Axe_Swing_Distant_04}}
 {id=WavePlayer_8, properties {soundWaveAssetPtr=/Game/Athena/Sounds/Weapons/Fire_Axe_Swing_Distant_05}}
 {id=WavePlayer_9, properties {soundWaveAssetPtr=/Game/Athena/Sounds/Weapons/Fire_Axe_Swing_Distant_06}}
]
properties {
 audioPropertiesSheet=/BRCosmetics/Sounds/Settings/PropertySheets/Pickaxe/APS_Cosmetic_Pickaxe_Swing
 maxDistance=3000
 soundClassObject=&Fort_WeaponMelee_PSM
 volumeMultiplier=0.25
 concurrencySet [&Player_Melee_Weapon_Conc]
}
